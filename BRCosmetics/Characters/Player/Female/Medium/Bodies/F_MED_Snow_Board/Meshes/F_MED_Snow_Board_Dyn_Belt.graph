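# Physics asset for the dangling belt piece on the female medium Snow Board player body, apparently for secondary motion.
# A box body on dyn_belt is constrained to a tiny capsule on dyn_belt_MSTR with asymmetric swing limits and tight twist.
# A large capsule on thigh_r gives the right leg a volume, linked to the master bone by a constraint with free swing and twist.
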PhysicsAsset /BRCosmetics/Characters/Player/Female/Medium/Bodies/F_MED_Snow_Board/Meshes/F_MED_Snow_Board_Dyn_Belt
bodies {
 dyn_belt [{shapeType=Box, extentX=12.7262, extentY=1.3998, extentZ=6.125, center=<6.0226 -0.0001 0>, rotation=<rot 0 0 20.4722>}]
 dyn_belt_MSTR [{shapeType=Capsule, length=1, radius=1}]
 thigh_r [{shapeType=Capsule, length=14.0621, radius=10.8708, center=<16.3851 0.1677 0.8582>, rotation=<rot -0.0001 77.3918 -89.9995>}]
}
constraints [
 {bone1Name=dyn_belt, bone2Name=dyn_belt_MSTR, swing1LimitDegrees=50, swing2LimitDegrees=25, twistLimitDegrees=5}
 {
  bone1Name=dyn_belt_MSTR
  bone2Name=thigh_r
  swing1LimitDegrees=45
  swing1Motion=Free
  swing2LimitDegrees=45
  swing2Motion=Free
  twistLimitDegrees=45
  twistMotion=Free
 }
]
