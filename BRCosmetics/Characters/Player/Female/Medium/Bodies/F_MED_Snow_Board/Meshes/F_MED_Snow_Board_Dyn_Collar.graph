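# Physics asset for the dynamic collar on the female medium snowboard body cosmetic, simulating collar motion around the head.
# Four collar bodies (front, back, left, right) each use three angled boxes forming curved panels, constrained to spine_05 with moderate swing and tight twist limits.
# A head capsule and small neck and spine capsules form a chain with free angular motion, giving the collar panels something to move against.
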PhysicsAsset /BRCosmetics/Characters/Player/Female/Medium/Bodies/F_MED_Snow_Board/Meshes/F_MED_Snow_Board_Dyn_Collar
bodies {
 collar_back [
  {shapeType=Box, extentX=16, extentY=4.5, extentZ=10, center=<7.4827 -2.2883 0>, rotation=<rot 0 -15.0008 0>}
  {shapeType=Box, extentX=16, extentY=4.5, extentZ=10, center=<7.8319 -1.2753 -5.2148>, rotation=<rot 0 -20 -34.9073>}
  {shapeType=Box, extentX=16, extentY=4.5, extentZ=10, center=<7.8623 -1.1925 5.1532>, rotation=<rot 0 -20 34.9073>}
 ]
 collar_front [
  {shapeType=Box, extentX=18.4046, extentY=4.5, extentZ=9.6196, center=<8.847 1.9475 -0.0001>, rotation=<rot 0 -10.9544 0>}
  {shapeType=Box, extentX=16, extentY=10, extentZ=3, center=<10.2562 -1.7764 -7.4077>, rotation=<rot 1.9229 -4.6317 -43.3266>}
  {shapeType=Box, extentX=16, extentY=10, extentZ=3, center=<10.2155 -2.1258 7.0352>, rotation=<rot -1.9229 -4.6317 43.3266>}
 ]
 collar_l [
  {shapeType=Box, extentX=16, extentY=10, extentZ=3, center=<8.2963 -0.0001 -3.4478>}
  {shapeType=Box, extentX=16, extentY=10, extentZ=3, center=<10.279 -5.333 -0.3302>, rotation=<rot -11.9502 -10.5671 42.0101>}
  {shapeType=Box, extentX=16, extentY=10, extentZ=3, center=<9.0347 9.4912 -0.8961>, rotation=<rot -0.855 9.2009 -34.0717>}
 ]
 collar_r [
  {shapeType=Box, extentX=16, extentY=10, extentZ=3, center=<8.2963 0 3.5091>}
  {shapeType=Box, extentX=16, extentY=10, extentZ=3, center=<10.2191 -5.5945 1.0493>, rotation=<rot 6.4377 -5.6099 -41.2147>}
  {shapeType=Box, extentX=16, extentY=10, extentZ=3, center=<9.2529 8.3062 1.6023>, rotation=<rot 0.855 9.2002 34.0709>}
 ]
 head [{shapeType=Capsule, length=2.5374, radius=9.0416, center=<4.5742 -0.5254 0.0046>, rotation=<rot 0.4599 -103.1582 89.8871>}]
 neck_01 [{shapeType=Capsule, length=2, radius=2}]
 spine_05 [{shapeType=Capsule, length=2, radius=2}]
}
constraints [
 {bone1Name=collar_r, bone2Name=spine_05, swing1LimitDegrees=25, swing2LimitDegrees=25, twistLimitDegrees=5}
 {bone1Name=collar_l, bone2Name=spine_05, swing1LimitDegrees=25, swing2LimitDegrees=25, twistLimitDegrees=5}
 {bone1Name=collar_front, bone2Name=spine_05, swing1LimitDegrees=25, swing2LimitDegrees=25, twistLimitDegrees=5}
 {bone1Name=collar_back, bone2Name=spine_05, swing1LimitDegrees=25, swing2LimitDegrees=25, twistLimitDegrees=5}
 {
  bone1Name=neck_01
  bone2Name=spine_05
  swing1LimitDegrees=45
  swing1Motion=Free
  swing2LimitDegrees=45
  swing2Motion=Free
  twistLimitDegrees=45
  twistMotion=Free
 }
 {
  bone1Name=head
  bone2Name=neck_01
  swing1LimitDegrees=45
  swing1Motion=Free
  swing2LimitDegrees=45
  swing2Motion=Free
  twistLimitDegrees=45
  twistMotion=Free
 }
]
