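# Post-process-style cosmetic AnimBlueprint for the F_MED_Snow_Board female body skeleton, adding secondary physics to belt and collar.
# Copies the attached parent's mesh pose, then additively nudges and rotates the collar_front bone.
# Two chained RigidBody nodes simulate belt and collar physics assets relative to thigh_r and spine_05, with custom gravity.
# Each update sets "Rigidbody disable" to the inverse of bRigidBodyEnabled, though no shown node reads it.
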AnimBlueprint /BRCosmetics/Characters/Player/Female/Medium/Bodies/F_MED_Snow_Board/Meshes/F_MED_Snow_Board_Skeleton_AnimBP
target_skeleton=&/F_MED_Snow_Board_Skeleton
graphs [
 AnimGraph {nodes [
   n1 = "Transform (Modify) Bone - Bone: collar_front"!(
    ComponentPose=Local To Component(LocalPose=Copy Pose From Mesh() {bUseAttachedParent=true})
    bAlphaBoolEnabled:bool=True
    Translation="0,1,0"
    Rotation="-0.005,-3,0.004"
    Scale="1,1,1"
   ) -> (Pose) {
    alphaInputType=Bool
    rotationMode=BMM_Additive
    rotationSpace=BCS_BoneSpace
    translationMode=BMM_Additive
    translationSpace=BCS_BoneSpace
    alphaBoolBlend {blendOption=Linear}
    boneToModify {boneName=collar_front}
    rotation=<rot -0.005 -3 0.004>
    translation {y=1}
   }
   n2 = RigidBody(ComponentPose=@n1) {
    bOverrideWorldGravity=true
    overridePhysicsAsset=&/F_MED_Snow_Board_Dyn_Belt
    simulationSpace=BaseBoneSpace
    alphaBoolBlend {blendOption=Linear}
    baseBoneRef {boneName=thigh_r}
    componentAppliedLinearAccClamp=<2000 2000 2000>
    componentLinearAccScale=<3 2 2>
    componentLinearVelScale=<3 2 2>
    overrideWorldGravity {z=-980}
   }
   n3 = RigidBody(ComponentPose=@n2) {
    bOverrideWorldGravity=true
    overridePhysicsAsset=&/F_MED_Snow_Board_Dyn_Collar
    simulationSpace=BaseBoneSpace
    alphaBoolBlend {blendOption=Linear}
    baseBoneRef {boneName=spine_05}
    componentAppliedLinearAccClamp=<1000 1000 1000>
    componentLinearAccScale=<2 1 1>
    componentLinearVelScale=<2 1 1>
    overrideWorldGravity {z=-980}
   }
   Output Pose(Result=Component To Local(ComponentPose=@n3)) {layerGroup=None}
  ]}
 EventGraph {nodes [
   Event BlueprintUpdateAnimation()
   > Set Rigidbody disable(Rigidbody disable:bool=NOT Boolean(A=Get bRigidBodyEnabled())) -> (Output_Get:bool)
  ]}
]
variables [Rigidbody disable {type=boolean, default=true}]
